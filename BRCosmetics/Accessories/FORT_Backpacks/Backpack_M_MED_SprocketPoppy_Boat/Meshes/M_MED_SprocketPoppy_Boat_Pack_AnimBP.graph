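# Anim Blueprint for the SprocketPoppy Boat backpack, fitting it to the character's body and jiggling its skull spine.
# Character spine socket transforms drive pelvis_xyz and spine_03_xyz in world space, while upper-spine yaw offsets and rolls the pack bone.
# The cached pose feeds five chained, spring-driven AnimDynamics bones; a second, softer chain is unconnected.
# Updated gravity, front-end alpha and sprint-scale variables are not read by the AnimGraph.
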AnimBlueprint /BRCosmetics/Accessories/FORT_Backpacks/Backpack_M_MED_SprocketPoppy_Boat/Meshes/M_MED_SprocketPoppy_Boat_Pack_AnimBP
target_skeleton=&/M_MED_SprocketPoppy_Boat_Pack_Skeleton
graphs [
 AnimGraph {nodes [
   n1 = Anim Dynamics - Bone: dyn_spineSkull_01(Alpha=1, ExternalForce="0,0,0") {
    angularDampingOverride=0.95
    angularSpringConstant=10000
    bAngularSpring=true
    bOverrideAngularDamping=true
    bUseGravityOverride=true
    bUsePlanarLimit=false
    alphaBoolBlend {blendOption=Linear}
    boundBone {boneName=dyn_spineSkull_01}
    componentAppliedLinearAccClamp=<300 300 300>
    componentLinearAccScale=<0.2 0.2 0.2>
    componentLinearVelScale=<0.3 0.3 0.3>
    gravityOverride {z=-600}
    physicsBodyDefinitions [{
      boundBone {boneName=dyn_spineSkull_01}
      boxExtents=<9 11 11>
      constraintSetup {angularLimitsMax=<3 2 0>, angularLimitsMin=<-3 -2 -2>, angularTarget=<0 0 0>}
      localJointOffset=<0.5 0 0>
     }]
   }
   n2 = Anim Dynamics - Bone: dyn_spineSkull_02(ComponentPose=@n1, Alpha=1, ExternalForce="0,0,0") {
    angularDampingOverride=0.95
    angularSpringConstant=10000
    bAngularSpring=true
    bOverrideAngularDamping=true
    bUseGravityOverride=true
    bUsePlanarLimit=false
    alphaBoolBlend {blendOption=Linear}
    boundBone {boneName=dyn_spineSkull_02}
    componentAppliedLinearAccClamp=<300 300 300>
    componentLinearAccScale=<0.2 0.2 0.2>
    componentLinearVelScale=<0.3 0.3 0.3>
    gravityOverride {z=-700}
    physicsBodyDefinitions [{
      boundBone {boneName=dyn_spineSkull_02}
      boxExtents=<9 11 11>
      constraintSetup {angularLimitsMax=<3 3 0>, angularLimitsMin=<-3 -3 -3>, angularTarget=<0 0 0>}
      localJointOffset=<0.5 0 0>
     }]
   }
   n3 = Anim Dynamics - Bone: dyn_spineSkull_03(ComponentPose=@n2, Alpha=1, ExternalForce="0,0,0") {
    angularDampingOverride=0.95
    angularSpringConstant=10000
    bAngularSpring=true
    bOverrideAngularDamping=true
    bUseGravityOverride=true
    bUsePlanarLimit=false
    alphaBoolBlend {blendOption=Linear}
    boundBone {boneName=dyn_spineSkull_03}
    componentAppliedLinearAccClamp=<260 260 260>
    componentLinearAccScale=<0.2 0.2 0.2>
    componentLinearVelScale=<0.3 0.3 0.3>
    gravityOverride {z=-800}
    physicsBodyDefinitions [{
      boundBone {boneName=dyn_spineSkull_03}
      boxExtents=<9 10 10>
      constraintSetup {angularLimitsMax=<3 9 0>, angularLimitsMin=<-3 -9 -10>, angularTarget=<0 0 0>}
      localJointOffset=<0.5 0 0>
     }]
   }
   n4 = Anim Dynamics - Bone: dyn_spineSkull_04(ComponentPose=@n3, Alpha=1, ExternalForce="0,0,0") {
    angularDampingOverride=0.95
    angularSpringConstant=10000
    bAngularSpring=true
    bOverrideAngularDamping=true
    bUseGravityOverride=true
    bUsePlanarLimit=false
    alphaBoolBlend {blendOption=Linear}
    boundBone {boneName=dyn_spineSkull_04}
    componentAppliedLinearAccClamp=<250 250 250>
    componentLinearAccScale=<0.2 0.2 0.2>
    componentLinearVelScale=<0.3 0.3 0.3>
    gravityOverride {z=-850}
    physicsBodyDefinitions [{
      boundBone {boneName=dyn_spineSkull_04}
      boxExtents=<9 10 10>
      constraintSetup {angularLimitsMax=<3 10 0>, angularLimitsMin=<-3 -10 -10>, angularTarget=<0 0 0>}
      localJointOffset=<0.5 0 0>
     }]
   }
   Anim Dynamics - Bone: dyn_spineSkull_05(ComponentPose=@n4, Alpha=1, ExternalForce="0,0,0") {
    angularDampingOverride=0.95
    angularSpringConstant=10000
    bAngularSpring=true
    bOverrideAngularDamping=true
    bUseGravityOverride=true
    bUsePlanarLimit=false
    alphaBoolBlend {blendOption=Linear}
    boundBone {boneName=dyn_spineSkull_05}
    componentAppliedLinearAccClamp=<250 250 250>
    componentLinearAccScale=<0.2 0.2 0.2>
    componentLinearVelScale=<0.3 0.3 0.3>
    gravityOverride {z=-950}
    physicsBodyDefinitions [{
      boundBone {boneName=dyn_spineSkull_05}
      boxExtents=<8.5 10 10>
      constraintSetup {angularLimitsMax=<3 12 0>, angularLimitsMin=<-3 -12 -10>, angularTarget=<0 0 0>}
      localJointOffset=<0.5 0 0>
     }]
   }
   n5 = Anim Dynamics - Bone: dyn_spineSkull_01(ComponentPose=Local To Component(LocalPose=Use cached pose 'base pose'()), Alpha=1, ExternalForce="0,0,0") {
    angularDampingOverride=0.97
    angularSpringConstant=15000
    bAngularSpring=true
    bOverrideAngularDamping=true
    bUseGravityOverride=true
    bUsePlanarLimit=false
    alphaBoolBlend {blendOption=Linear}
    boundBone {boneName=dyn_spineSkull_01}
    componentAppliedLinearAccClamp=<300 300 300>
    componentLinearAccScale=<0.3 0.3 0.3>
    componentLinearVelScale=<0.4 0.4 0.4>
    gravityOverride {z=-750}
    physicsBodyDefinitions [{
      boundBone {boneName=dyn_spineSkull_01}
      boxExtents=<9 11 11>
      constraintSetup {angularLimitsMax=<3 2 0>, angularLimitsMin=<-3 -2 -2>, angularTarget=<0 0 0>}
      localJointOffset=<0.6 0 0>
     }]
   }
   n6 = Anim Dynamics - Bone: dyn_spineSkull_02(ComponentPose=@n5, Alpha=1, ExternalForce="0,0,0") {
    angularDampingOverride=0.97
    angularSpringConstant=15000
    bAngularSpring=true
    bOverrideAngularDamping=true
    bUseGravityOverride=true
    bUsePlanarLimit=false
    alphaBoolBlend {blendOption=Linear}
    boundBone {boneName=dyn_spineSkull_02}
    componentAppliedLinearAccClamp=<300 300 300>
    componentLinearAccScale=<0.3 0.3 0.3>
    componentLinearVelScale=<0.4 0.4 0.4>
    gravityOverride {z=-800}
    physicsBodyDefinitions [{
      boundBone {boneName=dyn_spineSkull_02}
      boxExtents=<9 11 11>
      constraintSetup {angularLimitsMax=<3 3 0>, angularLimitsMin=<-3 -3 -3>, angularTarget=<0 0 0>}
      localJointOffset=<0.6 0 0>
     }]
   }
   n7 = Anim Dynamics - Bone: dyn_spineSkull_03(ComponentPose=@n6, Alpha=1, ExternalForce="0,0,0") {
    angularDampingOverride=0.97
    angularSpringConstant=15000
    bAngularSpring=true
    bOverrideAngularDamping=true
    bUseGravityOverride=true
    bUsePlanarLimit=false
    alphaBoolBlend {blendOption=Linear}
    boundBone {boneName=dyn_spineSkull_03}
    componentAppliedLinearAccClamp=<260 260 260>
    componentLinearAccScale=<0.3 0.3 0.3>
    componentLinearVelScale=<0.4 0.4 0.4>
    gravityOverride {z=-870}
    physicsBodyDefinitions [{
      boundBone {boneName=dyn_spineSkull_03}
      boxExtents=<9 10 10>
      constraintSetup {angularLimitsMax=<3 9 0>, angularLimitsMin=<-3 -9 -10>, angularTarget=<0 0 0>}
      localJointOffset=<0.6 0 0>
     }]
   }
   n8 = Anim Dynamics - Bone: dyn_spineSkull_04(ComponentPose=@n7, Alpha=1, ExternalForce="0,0,0") {
    angularDampingOverride=0.97
    angularSpringConstant=15000
    bAngularSpring=true
    bOverrideAngularDamping=true
    bUseGravityOverride=true
    bUsePlanarLimit=false
    alphaBoolBlend {blendOption=Linear}
    boundBone {boneName=dyn_spineSkull_04}
    componentAppliedLinearAccClamp=<250 250 250>
    componentLinearAccScale=<0.3 0.3 0.3>
    componentLinearVelScale=<0.4 0.4 0.4>
    gravityOverride {z=-930}
    physicsBodyDefinitions [{
      boundBone {boneName=dyn_spineSkull_04}
      boxExtents=<9 10 10>
      constraintSetup {angularLimitsMax=<3 10 0>, angularLimitsMin=<-3 -10 -10>, angularTarget=<0 0 0>}
      localJointOffset=<0.6 0 0>
     }]
   }
   n9 = Anim Dynamics - Bone: dyn_spineSkull_05(ComponentPose=@n8, Alpha=1, ExternalForce="0,0,0") {
    angularDampingOverride=0.97
    angularSpringConstant=15000
    bAngularSpring=true
    bOverrideAngularDamping=true
    bUseGravityOverride=true
    bUsePlanarLimit=false
    alphaBoolBlend {blendOption=Linear}
    boundBone {boneName=dyn_spineSkull_05}
    componentAppliedLinearAccClamp=<250 250 250>
    componentLinearAccScale=<0.3 0.3 0.3>
    componentLinearVelScale=<0.4 0.4 0.4>
    gravityOverride {z=-980}
    physicsBodyDefinitions [{
      boundBone {boneName=dyn_spineSkull_05}
      boxExtents=<8.5 10 10>
      constraintSetup {angularLimitsMax=<3 12 0>, angularLimitsMin=<-3 -12 -10>, angularTarget=<0 0 0>}
      localJointOffset=<0.6 0 0>
     }]
   }
   Output Pose(Result=Component To Local(ComponentPose=@n9)) {layerGroup=None}
   n10 = "Transform (Modify) Bone - Bone: pelvis_xyz"(
    ComponentPose=Local To Component(LocalPose=Input Pose())
    Alpha=1
    Translation=Get Pelvis Translation() -> (Pelvis Translation:Vector)
    Rotation="Get Pelvis  Rotation"() -> ("Pelvis  Rotation":Rotator)
    Scale="1,1,1"
   ) {
    rotationMode=BMM_Replace
    rotationSpace=BCS_WorldSpace
    translationMode=BMM_Replace
    translationSpace=BCS_WorldSpace
    alphaBoolBlend {blendOption=Linear}
    boneToModify {boneName=pelvis_xyz}
   }
   n11 = "Transform (Modify) Bone - Bone: spine_03_xyz"(
    ComponentPose=@n10
    Alpha=1
    Translation=Get Spine_03 Translation() -> (Spine_03 Translation:Vector)
    Rotation=Get Spine_03 Rotation() -> (Spine_03 Rotation:Rotator)
    Scale="1,1,1"
   ) {
    rotationMode=BMM_Replace
    rotationSpace=BCS_WorldSpace
    translationMode=BMM_Replace
    translationSpace=BCS_WorldSpace
    alphaBoolBlend {blendOption=Linear}
    boneToModify {boneName=spine_03_xyz}
   }
   n12 = "Transform (Modify) Bone - Bone: pack"(
    ComponentPose=@n11
    Translation=Get Backpack Translation() -> (Backpack Translation:Vector)
    Rotation=Get Backpack Rotate() -> (Backpack Rotate:Rotator)
   ) {
    rotationMode=BMM_Additive
    rotationSpace=BCS_BoneSpace
    translationMode=BMM_Additive
    translationSpace=BCS_BoneSpace
    alphaBoolBlend {blendOption=Linear}
    boneToModify {boneName=pack}
   }
   Save cached pose 'base pose'(Pose=Component To Local(ComponentPose=@n12)) {cacheName=base pose}
  ]}
 EventGraph {nodes [
   Event BlueprintUpdateAnimation()
   > n1 = Gravity Override State Function DEPRECATED(CharacterPartAnimInstance=Self-Reference(), GameTimeInSeconds=GetGameTimeInSeconds(), GravityOverrideParams=Get Gravity Override Params())
   > Set GravityOverride(GravityOverride=@n1)
   > n2 = Sequence!(execute) -> (then_0, then_1, then_2:Exec, then_3:Exec)
   n3 = Get SkelMesh()
   > Is Valid(InputObject=@n3)
   n4 = BreakTransform(InTransform=GetSocketTransform(self=@n3, InSocketName=spine_03, TransformSpace=RTS_World))
   > Set Pelvis Translation(Pelvis Translation:Vector=@n4) -> (Output_Get:Vector)
   > "Set Pelvis  Rotation"("Pelvis  Rotation":Rotator=@n4.Rotation) -> (Output_Get:Rotator)
   n5 = BreakTransform(InTransform=GetSocketTransform(self=@n3, InSocketName=spine_04, TransformSpace=RTS_World))
   > Set Spine_03 Translation(Spine_03 Translation:Vector=@n5) -> (Output_Get:Vector)
   > Set Spine_03 Rotation(Spine_03 Rotation:Rotator=@n5.Rotation) -> (Output_Get:Rotator)
   n6 = Branch(execute=@n2.then_1, Condition=Get bIsInFrontEnd())
   > Set Alpha_0(Alpha_0:float=0.6) -> (Output_Get:float)
   Set Alpha_0(execute=@n6.else, Alpha_0=1)
   n7 = SelectVector(A="0.9,0.3,0.3", B="0.4,0.3,0.3", bPickA=Get bIsTacticalSprinting())
   Set Component Linear Acc Scale(execute=@n2.then_2, Component Linear Acc Scale:Vector=@n7) -> (Output_Get:Vector)
   > Set Component Linear Vel Scale(Component Linear Vel Scale:Vector=@n7) -> (Output_Get:Vector)
   n8 = GetOwnerSkeletalMeshComponent()
   n9 = BreakRotator(InRot=BreakTransform(InTransform=GetDeltaTransformFromRefPose(self=@n8, BoneName=spine_05, BaseName=spine_04)).Rotation)
   Set Backpack Rotate(execute=@n2.then_3, Backpack Rotate:Rotator=MakeRotator(Roll=MapRangeClamped(Value=@n9.Yaw, InRangeA=0, InRangeB=-40, OutRangeB=45))) -> (Output_Get:Rotator)
   > Set Backpack Translation(Backpack Translation:Vector=MakeVector(Y=MapRangeClamped(Value=@n9.Yaw, InRangeA=0, InRangeB=-30, OutRangeB=-5))) -> (Output_Get:Vector)
   Event OnCharacterCustomizationLoaded()
   > n10 = Cast To PlayerPawn_Athena(Object=TryGetPawnOwner())
   > Set As Player Pawn Athena(As Player Pawn Athena:Player Pawn Athena Object Reference=@n10) -> (Output_Get:Player Pawn Athena Object Reference)
   > Set SkelMesh(SkelMesh=GetSkeletalMeshForPartType(self=@n10))
   TryGetPawnOwner()
  ]}
]
variables [
 Alpha_0 {type=number}
 As Player Pawn Athena {type=PlayerPawn_Athena_C}
 Backpack Rotate {type=Rotator}
 Backpack Translation {type=Vector}
 Component Linear Acc Scale {type=Vector}
 Component Linear Vel Scale {type=Vector}
 Gravity Override Params {type=GravityOverrideParamsStruct}
 GravityOverride {type=Vector}
 "Pelvis  Rotation" {type=Rotator}
 Pelvis Translation {type=Vector}
 SkelMesh {type=SkeletalMeshComponent}
 Spine_03 Rotation {type=Rotator}
 Spine_03 Translation {type=Vector}
]
